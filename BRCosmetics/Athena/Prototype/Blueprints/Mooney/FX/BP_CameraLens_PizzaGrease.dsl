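; BP_CameraLens_PizzaGrease is a Blueprint that, going by its name and emitter component, is a camera lens particle effect for pizza grease.
; Its construction script gets the emitter's particle system component and sets its translucent sort priority to 10, affecting draw order against other translucent objects.
; The script shows no other logic.
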(fn ConstructionScript
  (Rendering|SetTranslucentSortPriority (Variables|Emitter|GetParticleSystemComponent) 10))
